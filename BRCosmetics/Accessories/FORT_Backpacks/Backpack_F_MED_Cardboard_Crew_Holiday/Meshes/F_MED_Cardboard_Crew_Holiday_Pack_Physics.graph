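# Physics asset for the Cardboard Crew Holiday backpack cosmetic, providing bodies and joints for a bow with dangling ribbons.
# Front and back ribbons are three-box chains hanging from a capsule bow base, which chains through the pack body to root.
# Ribbon joints have tight swing and twist limits of a few degrees, while bow-to-pack and pack-to-root rotations are locked, confining angular motion to the ribbons.
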PhysicsAsset /BRCosmetics/Accessories/FORT_Backpacks/Backpack_F_MED_Cardboard_Crew_Holiday/Meshes/F_MED_Cardboard_Crew_Holiday_Pack_Physics
bodies {
 dyn_backRibbon_jnt_1 [{shapeType=Box, extentX=3, extentY=0.2, extentZ=2.5, center=<1.6016 0 0>, rotation=<rot 0 7.6554 0>}]
 dyn_backRibbon_jnt_2 [{shapeType=Box, extentX=2.5, extentY=0.5, extentZ=2.5, center=<1.767 -0.0029 0.1876>, rotation=<rot -0.0001 7.6554 0>}]
 dyn_backRibbon_jnt_3 [{shapeType=Box, extentX=3, extentY=0.2, extentZ=2.5, center=<2.0022 0.2096 0.3188>, rotation=<rot -0.0001 14.197 0>}]
 dyn_bow_base_jnt [{shapeType=Capsule, length=2, radius=2, rotation=<rot 90 0 0>}]
 dyn_frontRibbon_jnt_1 [{shapeType=Box, extentX=3, extentY=0.2, extentZ=2.5, center=<1.6016 0 0>}]
 dyn_frontRibbon_jnt_2 [{shapeType=Box, extentX=2.5, extentY=0.5, extentZ=2.5, center=<1.542 0 0>}]
 dyn_frontRibbon_jnt_3 [{shapeType=Box, extentX=3, extentY=0.2, extentZ=2.5, center=<1.8383 0 0>}]
 pack [{shapeType=Capsule, length=1, radius=1}]
 root [{shapeType=Capsule, length=1, radius=1}]
}
constraints [
 {bone1Name=dyn_frontRibbon_jnt_1, bone2Name=dyn_bow_base_jnt, swing1LimitDegrees=2, swing2LimitDegrees=2, twistLimitDegrees=2}
 {bone1Name=dyn_frontRibbon_jnt_2, bone2Name=dyn_frontRibbon_jnt_1, swing1LimitDegrees=5, swing2LimitDegrees=10, twistLimitDegrees=5}
 {bone1Name=dyn_frontRibbon_jnt_3, bone2Name=dyn_frontRibbon_jnt_2, swing1LimitDegrees=5, swing2LimitDegrees=5, twistLimitDegrees=3}
 {bone1Name=dyn_backRibbon_jnt_1, bone2Name=dyn_bow_base_jnt, swing1LimitDegrees=5, swing2LimitDegrees=5, twistLimitDegrees=5}
 {bone1Name=dyn_backRibbon_jnt_2, bone2Name=dyn_backRibbon_jnt_1, swing1LimitDegrees=5, swing2LimitDegrees=10, twistLimitDegrees=5}
 {bone1Name=dyn_backRibbon_jnt_3, bone2Name=dyn_backRibbon_jnt_2, swing1LimitDegrees=5, swing2LimitDegrees=5, twistLimitDegrees=3}
 {
  bone1Name=dyn_bow_base_jnt
  bone2Name=pack
  swing1LimitDegrees=45
  swing1Motion=Locked
  swing2LimitDegrees=45
  swing2Motion=Locked
  twistLimitDegrees=45
  twistMotion=Locked
 }
 {
  bone1Name=pack
  bone2Name=root
  swing1LimitDegrees=45
  swing1Motion=Locked
  swing2LimitDegrees=45
  swing2Motion=Locked
  twistLimitDegrees=45
  twistMotion=Locked
 }
]
