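# Animation Blueprint for the Cardboard Crew Holiday backpack, animating its back ribbon and adding physics motion.
# Each update sets SkydiveAlpha, despite its name a rotator, to one of two small opposite pitch or yaw offsets depending on bIsSkyDiving.
# Two ribbon joints add that rotation in bone space to the input pose.
# A RigidBody node then simulates the pack's physics asset relative to the pack bone, with damped component motion and overridden gravity.
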AnimBlueprint /BRCosmetics/Accessories/FORT_Backpacks/Backpack_F_MED_Cardboard_Crew_Holiday/Meshes/F_MED_Cardboard_Crew_Holiday_Pack_AnimBP
target_skeleton=&/F_MED_Cardboard_Crew_Holiday_Pack_Skeleton
graphs [
 AnimGraph {nodes [
   n1 = Get SkydiveAlpha() -> (SkydiveAlpha:Rotator)
   n2 = "Transform (Modify) Bone - Bone: dyn_backRibbon_jnt_2"(ComponentPose=Local To Component(LocalPose=Input Pose()), Alpha=1, Translation="0,0,0", Rotation=@n1, Scale="1,1,1") {
    rotationMode=BMM_Additive
    rotationSpace=BCS_BoneSpace
    alphaBoolBlend {blendOption=Linear}
    alphaScaleBiasClamp {bInterpResult=true}
    boneToModify {boneName=dyn_backRibbon_jnt_2}
   }
   n3 = "Transform (Modify) Bone - Bone: dyn_backRibbon_jnt_3"(ComponentPose=@n2, Alpha=1, Translation="0,0,0", Rotation=@n1, Scale="1,1,1") {
    rotationMode=BMM_Additive
    rotationSpace=BCS_BoneSpace
    alphaBoolBlend {blendOption=Linear}
    alphaScaleBiasClamp {bInterpResult=true}
    boneToModify {boneName=dyn_backRibbon_jnt_3}
   }
   n4 = RigidBody(ComponentPose=@n3) {
    bClampLinearTranslationLimitToRefPose=true
    bOverrideWorldGravity=true
    overridePhysicsAsset=&/F_MED_Cardboard_Crew_Holiday_Pack_Physics
    simulationSpace=BaseBoneSpace
    alphaBoolBlend {blendOption=Linear}
    baseBoneRef {boneName=pack}
    componentAppliedLinearAccClamp=<3000 3000 3000>
    componentLinearAccScale=<0.22 0.1 0.22>
    componentLinearVelScale=<0.22 0.1 0.22>
    overrideWorldGravity {z=-980}
   }
   Output Pose(Result=Component To Local(ComponentPose=@n4)) {layerGroup=None}
  ]}
 EventGraph {nodes [
   Event BlueprintUpdateAnimation()
   > Sequence()
   > Set SkydiveAlpha(SkydiveAlpha:Rotator=SelectRotator!(A:Rotator="0,-5,0", B:Rotator="0,5, 0", bPickA:bool=Get bIsSkyDiving()) -> (ReturnValue:Rotator)) -> (Output_Get:Rotator)
   TryGetPawnOwner()
  ]}
]
variables [SkydiveAlpha {type=Rotator}]
